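# Physics asset for the face accessory part of the female medium SkeleProbe HatOff body cosmetic, covering its dynamic hair bones.
# Fourteen dyn_hair bones each carry an identical small capsule, alongside one large capsule on the head.
# Each hair bone is constrained directly to the head rather than chained to another hair bone.
# All constraints share the same angular limits, with one swing axis much tighter than the other swing and twist.
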PhysicsAsset /BRCosmetics/Characters/Player/Female/Medium/Bodies/F_MED_SkeleProbe_HatOff/Meshes/Parts/F_MED_SkeleProbe_HatOff_FaceAcc_Physics
bodies {
 dyn_hair_01 [{
   shapeType=Capsule
   length=1
   radius=1
   shapeName="None, None, None, None, None, None, None, None, None, None, None, None, None, None"
   center=<0 0 1>
   rotation=<rot 0 0 90>
  }]
 dyn_hair_010 [{
   shapeType=Capsule
   length=1
   radius=1
   shapeName="None, None, None, None, None, None, None, None, None, None, None, None, None, None"
   center=<0 0 1>
   rotation=<rot 0 0 90>
  }]
 dyn_hair_011 [{
   shapeType=Capsule
   length=1
   radius=1
   shapeName="None, None, None, None, None, None, None, None, None, None, None, None, None, None"
   center=<0 0 1>
   rotation=<rot 0 0 90>
  }]
 dyn_hair_012 [{
   shapeType=Capsule
   length=1
   radius=1
   shapeName="None, None, None, None, None, None, None, None, None, None, None, None, None, None"
   center=<0 0 1>
   rotation=<rot 0 0 90>
  }]
 dyn_hair_013 [{
   shapeType=Capsule
   length=1
   radius=1
   shapeName="None, None, None, None, None, None, None, None, None, None, None, None, None, None"
   center=<0 0 1>
   rotation=<rot 0 0 90>
  }]
 dyn_hair_014 [{
   shapeType=Capsule
   length=1
   radius=1
   shapeName="None, None, None, None, None, None, None, None, None, None, None, None, None, None"
   center=<0 0 1>
   rotation=<rot 0 0 90>
  }]
 dyn_hair_02 [{
   shapeType=Capsule
   length=1
   radius=1
   shapeName="None, None, None, None, None, None, None, None, None, None, None, None, None, None"
   center=<0 0 1>
   rotation=<rot 0 0 90>
  }]
 dyn_hair_03 [{
   shapeType=Capsule
   length=1
   radius=1
   shapeName="None, None, None, None, None, None, None, None, None, None, None, None, None, None"
   center=<0 0 1>
   rotation=<rot 0 0 90>
  }]
 dyn_hair_04 [{
   shapeType=Capsule
   length=1
   radius=1
   shapeName="None, None, None, None, None, None, None, None, None, None, None, None, None, None"
   center=<0 0 1>
   rotation=<rot 0 0 90>
  }]
 dyn_hair_05 [{
   shapeType=Capsule
   length=1
   radius=1
   shapeName="None, None, None, None, None, None, None, None, None, None, None, None, None, None"
   center=<0 0 1>
   rotation=<rot 0 0 90>
  }]
 dyn_hair_06 [{
   shapeType=Capsule
   length=1
   radius=1
   shapeName="None, None, None, None, None, None, None, None, None, None, None, None, None, None"
   center=<0 0 1>
   rotation=<rot 0 0 90>
  }]
 dyn_hair_07 [{
   shapeType=Capsule
   length=1
   radius=1
   shapeName="None, None, None, None, None, None, None, None, None, None, None, None, None, None"
   center=<0 0 1>
   rotation=<rot 0 0 90>
  }]
 dyn_hair_08 [{
   shapeType=Capsule
   length=1
   radius=1
   shapeName="None, None, None, None, None, None, None, None, None, None, None, None, None, None"
   center=<0 0 1>
   rotation=<rot 0 0 90>
  }]
 dyn_hair_09 [{
   shapeType=Capsule
   length=1
   radius=1
   shapeName="None, None, None, None, None, None, None, None, None, None, None, None, None, None"
   center=<0 0 1>
   rotation=<rot 0 0 90>
  }]
 head [{shapeType=Capsule, length=11.3175, radius=8.428, shapeName=head_capsule}]
}
constraints [
 {bone1Name=dyn_hair_014, bone2Name=head, swing1LimitDegrees=30, swing2LimitDegrees=5, twistLimitDegrees=30}
 {bone1Name=dyn_hair_010, bone2Name=head, swing1LimitDegrees=30, swing2LimitDegrees=5, twistLimitDegrees=30}
 {bone1Name=dyn_hair_09, bone2Name=head, swing1LimitDegrees=30, swing2LimitDegrees=5, twistLimitDegrees=30}
 {bone1Name=dyn_hair_08, bone2Name=head, swing1LimitDegrees=30, swing2LimitDegrees=5, twistLimitDegrees=30}
 {bone1Name=dyn_hair_07, bone2Name=head, swing1LimitDegrees=30, swing2LimitDegrees=5, twistLimitDegrees=30}
 {bone1Name=dyn_hair_06, bone2Name=head, swing1LimitDegrees=30, swing2LimitDegrees=5, twistLimitDegrees=30}
 {bone1Name=dyn_hair_05, bone2Name=head, swing1LimitDegrees=30, swing2LimitDegrees=5, twistLimitDegrees=30}
 {bone1Name=dyn_hair_02, bone2Name=head, swing1LimitDegrees=30, swing2LimitDegrees=5, twistLimitDegrees=30}
 {bone1Name=dyn_hair_01, bone2Name=head, swing1LimitDegrees=30, swing2LimitDegrees=5, twistLimitDegrees=30}
 {bone1Name=dyn_hair_04, bone2Name=head, swing1LimitDegrees=30, swing2LimitDegrees=5, twistLimitDegrees=30}
 {bone1Name=dyn_hair_03, bone2Name=head, swing1LimitDegrees=30, swing2LimitDegrees=5, twistLimitDegrees=30}
 {bone1Name=dyn_hair_011, bone2Name=head, swing1LimitDegrees=30, swing2LimitDegrees=5, twistLimitDegrees=30}
 {bone1Name=dyn_hair_012, bone2Name=head, swing1LimitDegrees=30, swing2LimitDegrees=5, twistLimitDegrees=30}
 {bone1Name=dyn_hair_013, bone2Name=head, swing1LimitDegrees=30, swing2LimitDegrees=5, twistLimitDegrees=30}
]
